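; Waterfall FX actor for the Hades Cerberus gate that fades out its looping audio when destroyed.
; On Destroyed, it validity-checks and briefly fades the goopy impact loop, then does the same for the medium waterfall loop.
; The second fade is chained behind the first, so it is skipped if the impact component is invalid.
; BeginPlay, Tick, overlap and the construction script are empty.
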(event EventBeginPlay)

(event Collision|EventActorBeginOverlap
  (params (OtherActor :type Actor%20Object%20Reference)))

(event EventTick
  (params (DeltaSeconds :type Float%20%28single-precision%29)))

(event EventDestroyed
  (bind _spot_water_brimestone_waterfall_goopy_impact_loop_cue (Variables|Default|GetSpot_Water_Brimestone_Waterfall_Goopy_Impact_Loop_Cue))
  (bind _spot_water_brimestone_waterfall_medium_loop_cue (Variables|Default|GetSpot_Water_Brimestone_Waterfall_Medium_Loop_Cue))
  ;[1 1 1] Event Destroyed - Fade Out Audio
  (Utilities|IsValid _spot_water_brimestone_waterfall_goopy_impact_loop_cue
    (:Is%20Valid
      (if (Utilities|IsValid _spot_water_brimestone_waterfall_goopy_impact_loop_cue :!class /Script/Engine.KismetSystemLibrary)
        (Audio|Components|Audio|FadeOut _spot_water_brimestone_waterfall_goopy_impact_loop_cue 0.5 0.0 "Linear" :!class /Script/Engine.AudioComponent)
        (Utilities|IsValid _spot_water_brimestone_waterfall_medium_loop_cue
          (:Is%20Valid
            (if (Utilities|IsValid _spot_water_brimestone_waterfall_medium_loop_cue :!class /Script/Engine.KismetSystemLibrary)
              (Audio|Components|Audio|FadeOut _spot_water_brimestone_waterfall_medium_loop_cue 0.5 0.0 "Linear" :!class /Script/Engine.AudioComponent))))))))

(fn ConstructionScript)
