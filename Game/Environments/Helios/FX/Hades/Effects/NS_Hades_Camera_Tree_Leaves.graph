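# GPU Niagara effect for the Hades environment that kicks up short-lived tree leaves around the camera.
# A brief spawn burst seeds sprite and mesh leaves on sampled landscape, filtered by a mask texture.
# Negative-strength point attraction near the camera pushes leaves; a latched terrain-height flag enables curl noise and wind and turns sprite orientation from ground normal toward camera.
# A second DEBUG sprite emitter is disabled.
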NiagaraSystem /Game/Environments/Helios/FX/Hades/Effects/NS_Hades_Camera_Tree_Leaves
dependencies {
 usedDataInterfaces [
  &NiagaraDataInterfaceCamera
  &NiagaraDataInterfaceCurve
  /Script/Niagara.NiagaraDataInterfaceTexture
  &NiagaraDataInterfaceCollisionQuery
  &NiagaraDataInterfaceColorCurve
 ]
 usedDynamicInputs []
 usedModules [&SystemState]
}
emitters {
 DEBUG {
  bEnabled=false
  simTarget=GPUComputeSim
  emitterUpdateScript [
   &EmitterState()
   &SpawnRate(SpawnRate=350000, Use Spawn Probability=false)
  ]
  particleSpawnScript [
   &AddVelocity(Velocity=<0 50 0>)
   &InitializeParticle(
    Color Maximum=<rgba 0 0.1413 0.0949 1>
    Color Minimum=<rgba 0 1 0.2085 1>
    Color Mode=::NewEnumerator2
    Lifetime Max=0.5
    Lifetime Min=0.5
    Lifetime Mode=::NewEnumerator1
    Mesh Scale Mode=::NewEnumerator1
    Mesh Uniform Scale Max=10
    Mesh Uniform Scale Min=6
    Sprite Size Mode=::NewEnumerator1
    Sprite UV Mode=::NewEnumerator1
    Uniform Sprite Size Max=125
    Uniform Sprite Size Min=80
    UsePositionOffset=false
    Write Material Random=true
   )
   &ShapeLocation(
    Cylinder Radius=50000
    MaskSpawnGroup=false
    Override Seed=false
    Shape Primitive=::NewEnumerator1
    Surface Only=true
    Surface Only Band Thickness=1
    Use Endcaps In Surface Only Mode=false
    UseOwnerScale=true
   )
   &CameraOffset(Camera Offset Amount=5)
   &self:ScratchModule_08(
    BoxCenter:Vector3f=<36500 -79031.6406 1996>
    BoxExtents:Vector3f=<100000 100000 998>
    TextureSample=DataInterface() {propertyValues {
      Texture=/Game/Environments/Helios/FX/Hades/Textures/T_Hades_WaterSpawnMask_Position_01
      TextureUserParameter {parameter {name=None, typeDefHandle {registeredTypeIndex=2}, varData []}}
     }}
   ) {module=ScratchModule_08}
   &KillParticlesInVolume(
    Box Size:Vector3f=<15000 15000 250000>
    Kill Shape=::NewEnumerator1
    Volume Origin:Vector3f=&AddVectorToPosition(Position=<49038 -89768 3140>, Vector=<0 0 0>)
   )
   &self:RainTextureMasking_01(
    AlphaThreshold=0.5
    BoxCenter:NiagaraPosition=<36500 -79031.6406 3473.7344>
    BoxExtents:Vector3f=<100000 100000 1>
    Position:NiagaraPosition=$Particles.Position
    RejectionTexture=DataInterface() {propertyValues {
      Texture=/Game/Environments/Helios/FX/Hades/Textures/T_Hades_LeafSpawnMask_01
      TextureUserParameter {parameter {name=None, typeDefHandle {registeredTypeIndex=2}, varData []}}
     }}
   ) {enabled=false, module=RainTextureMasking_01}
   &KillParticles(Kill Particles=&Boolean_NotOperation(Bool=$Output.ScratchModule_08.Valid)) {enabled=false}
  ]
  particleUpdateScript [
   &ParticleState()
   &CurlNoiseForce(
    Curl Noise Cone Mask Angle=25
    Mask Curl Noise=true
    Noise Frequency=&RandomRangeFloat(Minimum=2, Maximum=4)
    Noise Strength=&Multiply_Float(A=&RandomRangeFloat(Minimum=25, Maximum=100), B=1)
    Pan Noise=true
   )
   &WindForce(Use Depth Buffer=false, Use Landscape=false, Wind Speed=<0 25 0>)
   &Drag(Drag=0.1, Use Linear Drag=true, Use Rotational Drag=true)
   &SolveForcesAndVelocity(
    Acceleration Limit=25
    Clamp Velocity=true
    Force:Vector3f=&Multiply_Vector(A=$Transient.PhysicsForce, B=<1 1 0>)
    Limit Acceleration=true
    Speed Limit=25
    Velocity:Vector3f=&Multiply_Vector(A=$Particles.Velocity, B=<1 1 0>)
   )
   &CalculateAccurateVelocity() {enabled=false}
   &CurlNoiseForce(
    Curl Noise Cone Mask Angle=80
    Curl Noise Cone Mask Falloff Angle=90
    Mask Curl Noise=true
    Noise Frequency=&RandomRangeFloat(Minimum=0.25, Maximum=1)
    Noise Strength=&Multiply_Float(A=&RandomRangeFloat(Minimum=35, Maximum=100), B=1)
    Pan Noise=false
    Pan Noise Field=<0 0 2>
   ) {module=CurlNoiseForce001}
   /Niagara/Modules/Math/AvoidDistanceFieldSurfaces_GPU(
    Avoid Nearest Surface=true
    Avoid Oncoming Surfaces=true
    Falloff Distance=1
    Max Trace Distance=1
    Nearest Surface Avoidance Strength=1
    Obstactle Avoidance Strength=1
    Trace Vector:Vector3f=&Multiply_Vector(A=$Particles.Velocity, B=<1 1 0>)
   ) {enabled=false}
   &SolveForcesAndVelocity(
    Acceleration Limit=5
    Clamp Velocity=true
    Force:Vector3f=&Multiply_Vector(A=$Transient.PhysicsForce, B=<1 1 0>)
    Limit Acceleration=true
    Speed Limit=100
    Velocity:Vector3f=&Multiply_Vector(A=$Particles.Velocity, B=<1 1 0>)
   ) {module=SolveForcesAndVelocity001}
   &RegionMask(
    Invert Region=true
    Offset:Vector3f=<49038 -89768 3140>
    Offset Space:ENiagaraCoordinateSpace=::Local
    Region Dimensions:Vector3f=<15000 15000 15000>
    Region Type:ENiagaraDebugDrawShape=::NewEnumerator1
    UseOffset=true
    UseRotation=false
   )
   &ScaleColor(
    Scale Alpha=&FloatFromCurve(
     FloatCurve=DataInterface() {propertyValues {Curve {keys [
         {interpMode=RCIM_Cubic, tangentMode=RCTM_User, time=0, value=0}
         {interpMode=RCIM_Cubic, tangentMode=RCTM_User, time=0.5, value=1}
         {tangentMode=RCTM_User, time=1, value=0}
        ]}}}
     Scale Curve=&Multiply_Float(A=0.25, B=$StackContext.RegionMask.Float)
    )
    Scale RGB=&Multiply_VectorByFloat(Vector=<1 1 1>, Float=&RandomRangeFloat(Minimum=1, Maximum=4))
    ScaleA=true
    ScaleRGB=true
   ) {module=ScaleColor001}
   /SetVariables(
    Particles.Linear Color=&MultiplyLinearColor(
     Linear Color=$Particles.Color
     Scale Factor=&MakeVector4_FromFloat(Float=&FloatFromCurve(
       FloatCurve=DataInterface() {propertyValues {Curve {keys [
           {interpMode=RCIM_Cubic, tangentMode=RCTM_User, time=0, value=0}
           {interpMode=RCIM_Cubic, tangentMode=RCTM_User, time=0.5, value=1}
           {tangentMode=RCTM_User, time=1, value=0}
          ]}}}
       Scale Curve=$System.ColorIntensityMult
      ))
    )
    Particles.Linear Color001:LinearColor=&MultiplyLinearColor(
     Linear Color=&ColorFromCurve(ColorCurve=DataInterface() {propertyValues {
        AlphaCurve {keys [<at 0 0 RCIM_Cubic>,<at 0.1218 1>,<at 1 0 RCIM_Cubic>]}
        BlueCurve {keys [<at 0 0.0688 RCIM_Cubic>,<at 1 1 RCIM_Cubic>]}
        GreenCurve {keys [<at 0 1 RCIM_Cubic>,<at 1 1 RCIM_Cubic>]}
        RedCurve {keys [<at 0 0 RCIM_Cubic>,<at 1 1 RCIM_Cubic>]}
       }})
     Scale Factor=&MakeVector4_FromFloat(Float=&FloatFromCurve(
       FloatCurve=DataInterface() {propertyValues {Curve {keys [<at 0 1>,<at 1 0>]}}}
       Scale Curve=&Multiply_Float(A=$System.ColorIntensityMult, B=4)
      ))
    )
   ) {isSetParameters=true}
   &CameraOffset() {module=CameraOffset001}
  ]
  rendererClasses [&NiagaraSpriteRendererProperties]
  renderers [{rendererClass=&NiagaraSpriteRendererProperties}]
 }
 Hades_Leaves_LoopSpawn {
  simTarget=GPUComputeSim
  emitterUpdateScript [
   /SetVariables() {isSetParameters=true}
   /SetVariables(
    Emitter.CameraForward:Vector3f=&ReturnCameraVector(Camera Property=::NewEnumerator1, Coordinate Space=::Simulation, CameraQuery=$Emitter.CameraRead)
    Emitter.CameraPosition:Vector3f=&ReturnCameraVector(Camera Property=::NewEnumerator0, Coordinate Space=::World, CameraQuery=$Emitter.CameraRead)
    Emitter.CameraRight:Vector3f=&ReturnCameraVector(Camera Property=::NewEnumerator3, Coordinate Space=::Simulation, CameraQuery=$Emitter.CameraRead)
    Emitter.CameraUp:Vector3f=&ReturnCameraVector(Camera Property=::NewEnumerator2, Coordinate Space=::Simulation, CameraQuery=$Emitter.CameraRead)
   ) {isSetParameters=true}
   &EmitterState(Life Cycle Mode=::NewEnumerator1, Loop Behavior=::NewEnumerator1, Loop Duration Mode=::NewEnumerator1, UseLoopDelay=false)
   &SpawnRate(
    SpawnRate=&FloatFromCurve(FloatCurve=DataInterface() {propertyValues {Curve {keys [<at 0 0>,<at 0.1 1>,<at 0.2 0>]}}}, CurveIndex=$Emitter.Age, Scale Curve=50000)
    Use Spawn Probability=false
   )
  ]
  particleSpawnScript [
   &InitializeParticle(
    Color=&RandomRangeLinearColor(Minimum=<rgba 0.0775 0.0585 0.1042 1>, Maximum=<rgba 0.172 0.1541 0.2344 1>)
    Lifetime=&RandomRangeFloat(Minimum=5, Maximum=5)
    Mass=3
    Mass Mode=::NewEnumerator1
    Mesh Scale Mode=::NewEnumerator1
    Mesh Uniform Scale Max=0.1
    Mesh Uniform Scale Min=0.08
    Sprite Rotation Mode=::NewEnumerator2
    Sprite Size Mode=::NewEnumerator1
    Uniform Sprite Size Max=250
    Uniform Sprite Size Min=80
    UsePositionOffset=false
    Write Material Random=true
   ) {module=InitializeParticle001}
   /SetVariables(Particles.Terrain Offset=-5) {isSetParameters=true}
   &ShapeLocation(MaskSpawnGroup=false, Offset Mode=::NewEnumerator0, Override Seed=false, Sphere Radius=5000, UseOwnerScale=true) {module=ShapeLocation001}
   &self:Sample_Landscape_01() {module=Sample_Landscape_01}
   /SetVariables(Particles.Position=&Add_Vector(
     A=Linked() {linkedVariable {name=Particles.Position, type=NiagaraPosition}}
     B=&MakeVector(X=0, Y=0, Z=$Particles.Terrain Offset)
    )) {isSetParameters=true}
   /SetVariables(Particles.New Vector 3f:Vector3f=<0 0 1>) {isSetParameters=true}
   &CameraOffset(Camera Offset Amount=25)
   &SubUVAnimation(SubUV Animation Mode:ENiagara_SubUVLookupMode=::NewEnumerator2) {module=SubUVAnimation001}
   /SetVariables(Particles.SubImageIndex=&RandomRangeFloat(Minimum=0, Maximum=63)) {isSetParameters=true}
   /SetVariables() {isSetParameters=true}
   &AddVelocity(Velocity=<0 0 2500>) {enabled=false}
   &InitialMeshOrientation(
    Mesh Orientation Mode=::NewEnumerator1
    Rotation=&RandomRangeVector(Minimum=<0 0 0>, Maximum=<0 0 1>)
    Rotation Coordinate Space=::NewEnumerator2
    Use Rotation Vector=true
   ) {enabled=false}
   &self:RainTextureMasking(
    AlphaThreshold=0.5
    BoxCenter:NiagaraPosition=<36400 -79031.6406 3473.7344>
    BoxExtents:Vector3f=<100000 100000 1>
    Position:NiagaraPosition=$Particles.Position
    RejectionTexture=DataInterface() {propertyValues {
      Texture=&T_Hades_LeafSpawnMask_01
      TextureUserParameter {parameter {name=None, typeDefHandle {registeredTypeIndex=2}, varData []}}
     }}
   ) {module=RainTextureMasking}
   &InitialMeshOrientation(
    Mesh Axis To Orient=<0 0 1>
    Mesh Orientation Mode=::NewEnumerator2
    Orientation Vector=$Particles.Landscape Normal
    Rotation=&MakeVector(X=0, Y=0, Z=&RandomRangeFloat(Minimum=0, Maximum=1))
    Use Rotation Vector=true
   ) {module=InitialMeshOrientation001}
  ]
  particleUpdateScript [
   &self:Sample_Landscape_UPdate_01(Bias:NiagaraFloat=$Particles.Terrain Offset, New Landscape Sample:NiagaraDataInterfaceLandscape=$User.Landscape) {module=Sample_Landscape_UPdate_01}
   &ScaleColor(
    Scale Alpha=&FloatFromCurve(FloatCurve=DataInterface() {propertyValues {Curve {keys [<at 0.8 1>,<at 1 0>]}}})
    Scale RGB=&MakeVectorFromLinearColorRGB(Color=&MultiplyLinearColor(Linear Color=$Particles.Initial.Color, Scale Factor=&MakeVector4_FromFloat(Float=1)))
    ScaleA=true
    ScaleRGB=true
   )
   /SetVariables(Particles.Has Collided With Terrain:NiagaraBool=&SetBoolByFloatComparison(
     A=&Add_Float(A=$Output.Sample_Landscape_UPdate_01.Terrain Height, B=5)
     B=&MakeFloatFromVector(Vector=Linked() {linkedVariable {name=Particles.Position, type=NiagaraPosition}}, Channel=::NewEnumerator2)
     Comparison Type=::NewEnumerator2
    )) {isSetParameters=true}
   /Niagara/Modules/Update/Forces/PointAttractionForce(
    Attraction Radius:NiagaraFloat=&Waveform(
     Waveform Count=::NewEnumerator1
     Global Amplitude Scale=80
     "Waveform [1]"=::NewEnumerator0
     "[1] Frequency"=5
     "[1] Amplitude Min/Max"=<0.4 1>
     "[2] Average"=false
     "[2] Input Position"=0
     "[2] Phase"=$Emitter.Age
     "[2] Phase Scale"=1
     "[2] Compound Frequency"=2
     "[2] Frequency"=2
     "[2] Amplitude Min/Max"=<0 1>
     "[2] Scale"=1
     "[2] Bias"=0
     "[2] Exponent"=1
     "[2] Pulse Duration"=0.1
     "[2] Clip Triangle Min/Max"=<0 0>
     "[2] Random Seed"=0
    )
    AttractionStrength:NiagaraFloat=&Multiply_Float(
     A=&FloatFromCurve(FloatCurve=DataInterface() {propertyValues {Curve {keys [<at 0 1>,<at 0.25 0>]}}}, Scale Curve=-80)
     B=1
    )
    AttractorPosition:Vector3f=&Add_Vector(
     A=$Emitter.CameraPosition
     B=&Add_Vector(
      A=&Add_Vector(
       A=&Multiply_VectorByFloat(Vector=$Particles.New Vector 3f, Float=-140)
       B=&Multiply_VectorByFloat(Vector=$Emitter.CameraForward, Float=300)
      )
      B=&Multiply_VectorByFloat(Vector=$Emitter.CameraRight, Float=-50)
     )
    )
    Falloff Exponent=1
    Module.Kill Within Radius=false
    Module.Use Falloff=false
   ) {module=PointAttractionForce001}
   /SetVariables(Particles.Normalized Velocity:Vector3f=&NormalizeVector(Vector To Normalize=$Particles.Velocity)) {isSetParameters=true}
   /SetVariables(Particles.Is Active:NiagaraBool=&Boolean_OrOperation(
     Bool1=$Particles.Is Active
     Bool2=&SetBoolByFloatComparison(
      A=&MakeFloatFromVector(Vector=Linked() {linkedVariable {name=Particles.Position, type=NiagaraPosition}}, Channel=::NewEnumerator2)
      B=$Output.Sample_Landscape_UPdate_01.Terrain Height
      Comparison Type=::NewEnumerator6
     )
    )) {isSetParameters=true}
   &CurlNoiseForce(
    Noise Frequency=25
    Noise Strength=&MakeCustomFloatFromBool(
     Bool=$Particles.Is Active
     True Float=&FloatFromCurve(FloatCurve=DataInterface() {propertyValues {Curve {keys [<at 0 1>,<at 1 0.1>]}}}, Scale Curve=5000)
     False Float=0
    )
    Pan Noise=false
   ) {module=CurlNoiseForce001}
   &Drag(
    Drag=&MakeCustomFloatFromBool(
     Bool=$Particles.Is Active
     True Float=&FloatFromCurve(
      FloatCurve=DataInterface() {propertyValues {Curve {keys [
          {interpMode=RCIM_Cubic, tangentMode=RCTM_User, time=0, value=1}
          {interpMode=RCIM_Cubic, tangentMode=RCTM_User, time=0.2, value=0.95}
          {interpMode=RCIM_Cubic, tangentMode=RCTM_User, time=0.5, value=0}
         ]}}}
      Scale Curve=0
     )
     False Float=0
    )
    Use Linear Drag=true
    Use Rotational Drag=false
   )
   &GravityForce(Gravity=&MakeVector(
     X=0
     Y=0
     Z=&Multiply_Float(
      A=&FloatFromCurve(
       FloatCurve=DataInterface() {propertyValues {Curve {keys [
           <at 0 0>
           {interpMode=RCIM_Cubic, tangentMode=RCTM_User, time=0.02, value=2}
           <at 0.1 0>
           {interpMode=RCIM_Cubic, tangentMode=RCTM_User, time=0.2, value=-0.9}
           <at 0.5 -0.25>
           <at 0.8 0.1>
           <at 1 0.1>
          ]}}}
       Scale Curve=200
      )
      B=1
     )
    )) {module=GravityForce001}
   /SetVariables(Particles.Sprite Orientation:Vector3f=&Lerp_Vector(
     A=$Particles.Landscape Normal
     B=&ReturnCameraVector(Camera Property=::NewEnumerator4, Coordinate Space=::Simulation, Origin Position=$Particles.Position, CameraQuery=$Emitter.CameraRead)
     Alpha=&MakeCustomFloatFromBool(Bool=$Particles.Is Active, True Float=1, False Float=0)
    )) {isSetParameters=true}
   &SolveForcesAndVelocity(Clamp Velocity=true, Limit Acceleration=false, Speed Limit=800)
   &self:ScratchModule_05(Terrain Height:NiagaraFloat=$Output.Sample_Landscape_UPdate_01.Terrain Height, Threshold=25) {module=ScratchModule_05}
   &SubUVAnimation(End Frame=64, SubUV Animation Mode:ENiagara_SubUVLookupMode=::NewEnumerator1, SubUV Loop Count=4)
   &ParticleState(DeltaTime:NiagaraFloat=&MakeCustomFloatFromBool(
     Bool=&SetBoolByFloatComparison(
      A=&VectorLength(Vector=$Particles.Velocity, Fallback Vector=<0 0 0>, Fallback Threshold=1, Fallback Length=0)
      B=0
     )
     True Float=$Engine.DeltaTime
     False Float=0
    ))
   &WindForce(
    Offset By Time:Vector3f=<0 0 -10>
    Scale=5
    Use Ground Mask=true
    Wind Speed=&MakeVector(
     X=&MakeCustomFloatFromBool(
      Bool=$Particles.Is Active
      True Float=&FloatFromCurve(FloatCurve=DataInterface() {propertyValues {Curve {keys [<at 0.25 1>,<at 0.5 0.25>]}}}, Scale Curve=250)
      False Float=0
     )
     Y=0
     Z=0
    )
   ) {module=WindForce001}
   &AerodynamicDrag(
    Aerodynamic Drag=0.1
    Aerodynamic Rotational Drag=0.1
    Drag Shape Curve=DataInterface() {propertyValues {Curve {keys [
        {interpMode=RCIM_Cubic, tangentMode=RCTM_User, time=0, value=0.1}
        {arriveTangent=2.0799, interpMode=RCIM_Cubic, leaveTangent=2.0799, tangentMode=RCTM_User, time=1, value=1}
       ]}}}
    Lift Shape Curve=DataInterface() {propertyValues {Curve {keys [
        {interpMode=RCIM_Cubic, tangentMode=RCTM_User, time=0, value=0}
        {interpMode=RCIM_Cubic, tangentMode=RCTM_User, time=0.5, value=1}
        {tangentMode=RCTM_User, time=1, value=0}
       ]}}}
    Pivot Offset Maximum:Vector3f=<1 1 1>
    Pivot Offset Minimum Length:Vector3f=<0 0 0>
    Rotation Strength:NiagaraFloat=&RandomRangeFloat(Minimum=15, Maximum=25)
    Use Linear Drag=true
    Use Rotational Drag=true
   )
   &SolveForcesAndVelocity(Clamp Velocity=true, Limit Acceleration=false, Speed Limit=250) {module=SolveForcesAndVelocity001}
   &AlignSpriteToMeshOrientation(
    Mesh Orientation Relative Sprite Alignment Vector=<1 0 0>
    Mesh Orientation Relative Sprite Facing Vector:Vector3f=<1 1 1>
    SetAlignment=true
    SetFacing=true
   )
   &ScaleMeshSize(Scale Factor=&VectorFromFloat(Value=&FloatFromCurve(
      FloatCurve=DataInterface() {propertyValues {Curve {keys [
          {interpMode=RCIM_Cubic, tangentMode=RCTM_User, time=0, value=0.825}
          {interpMode=RCIM_Cubic, tangentMode=RCTM_User, time=0.5, value=0.75}
          {tangentMode=RCTM_User, time=1, value=0}
         ]}}}
      Scale Curve=&ClampFloat(Float=&DistanceBetweenPositions(A=Linked() {linkedVariable {name=Emitter.CameraPosition, type=Vector3f}}, B=$Particles.Position))
     )))
   &ScaleSpriteSize(Scale Sprite Size Mode=::NewEnumerator0)
   &KillParticles(Kill Particles=&self:ScratchDynamicInput())
   /SetVariables(Particles.SpriteColor=&MultiplyLinearColor(Linear Color=$Particles.Color, Scale Factor=&MakeVector4_FromFloat(Float=1))) {isSetParameters=true}
   &DynamicMaterialParameters(
    Index 0 Param 1=2
    Module.Index0Param0Write=true
    Module.Index0Param1Write=true
    Module.Index0Param2Write=true
    Module.Index0Param3Write=true
   )
  ]
  rendererClasses [&NiagaraSpriteRendererProperties, &NiagaraMeshRendererProperties]
  renderers [
   {rendererClass=&NiagaraSpriteRendererProperties}
   {rendererClass=&NiagaraMeshRendererProperties, rendererIndex=1}
  ]
 }
}
summary {
 systemName=NS_Hades_Camera_Tree_Leaves
 emitters [
  {
   emitterName=Hades_Leaves_LoopSpawn
   simTarget=GPUComputeSim
   rendererClasses [&NiagaraSpriteRendererProperties, &NiagaraMeshRendererProperties]
  }
  {bEnabled=false, emitterName=DEBUG, simTarget=GPUComputeSim, rendererClasses [&NiagaraSpriteRendererProperties]}
 ]
 userVariables [
  User.Alpha_AttractionStrength {type=NiagaraFloat, defaultValue=0, description=""}
  User.Landscape {
   type=NiagaraDataInterfaceLandscape
   description=""
   defaultValue {
    struct=NiagaraExt_VariableValue_DataInterface
    value {
     dataInterface=&self:NiagaraDataInterfaceLandscape_0
     dataInterfaceClass=&NiagaraDataInterfaceLandscape
    }
   }
  }
  User.Player Position {type=Vector3f, description="", defaultValue {struct=Vector3f, value=<0 0 0>}}
 ]
}
