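# Translucent, unlit, two-sided Hades fireball material for Niagara mesh particles.
# Two panning flame textures, each UV-distorted by scrolling noise and refined, sum into a mask that blends colors, tinted by Fresnel-blended emissive and hue shifts.
# That mask also drives opacity, softened by inverse Fresnel, with optional platform-switched depth fade and vertical gradient.
# A WPO branch builds vertex-normal displacement from a panning normal texture and scrolling ramp.
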Material /Game/Environments/Helios/FX/Hades/Materials/M_Hades_FireBall
graph {
 properties {
  bUsedWithNiagaraMeshParticles=true
  blendMode=BLEND_Translucent
  pixelDepthOffsetMode=PDOM_Legacy
  shadingModel=MSM_Unlit
  translucencyLightingMode=TLM_SurfacePerPixelLighting
  twoSided=true
  usedShadingModels=Unlit
 }
 nodes [
  n1 = TextureCoordinate()
  n2 = VectorParameter() {group=WPO, parameterName=WPO_UVControl, channelNames {a=VPan, b=UPan, g=VScale, r=UScale}, defaultValue=<rgba 0.1 1 0.1 1>}
  n3 = ComponentMask(None=@n2) {g=true, r=true}
  n4 = TextureSampleParameter2D(UVs=Panner(Coordinate=Multiply(A=@n1, B=@n3), Speed=AppendVector(A=@n2.UPan, B=@n2.UPan))) {parameterName=WPO_Texture, samplerType=SAMPLERTYPE_Normal, texture=/Game/Global/Textures/Default/Tileable_Patterns/T_Water2_N}
  Add(A=@n4.RGB)
  n5 = ScalarParameter() {parameterName=Hue Shift Percentage}
  n6 = TextureCoordinate()
  n7 = VectorParameter() {parameterName=Fire1_Distort_UVControl, channelNames {a=VPan, b=UPan, g=VScale, r=UScale}, defaultValue=<rgba 1 1 1 1>}
  n8 = AppendVector(A=@n7.UScale, B=@n7.UScale)
  n9 = TextureSampleParameter2D(UVs=Panner(Coordinate=Multiply(A=@n6, B=@n8), Speed=AppendVector(A=@n7.UPan, B=@n7.UPan))) {parameterName=Fire1_Distort_Texture, texture=/Game/Effects/Fort_Effects/Emotes/UltraLight/Texture/T_SoftNoise_BC7}
  n10 = AppendVector(A=@n9.R, B=@n9.R)
  n11 = ScalarParameter() {defaultValue=0.02, parameterName=Distort1_Mult}
  n12 = VectorParameter() {parameterName=Fire1_Texture_UVControl, channelNames {a=VPan, b=UPan, g=VScale, r=UScale}, defaultValue=<rgba 1 1 0.1 1>}
  n13 = Multiply(A=TextureCoordinate(), B=AppendVector(A=@n12.UScale, B=@n12.UScale))
  n14 = AppendVector(A=@n12.UPan, B=@n12.UPan)
  n15 = TextureSampleParameter2D(UVs=Add(A=Multiply(A=@n10, B=@n11), B=Panner(Coordinate=@n13, Speed=@n14))) {
   parameterName=Fire1_Texture
   samplerType=SAMPLERTYPE_LinearGrayscale
   texture=/Game/Effects/Fort_Effects/Emotes/UltraLight/Texture/T_FluidFlameTile_Greyscale
  }
  n16 = VectorParameter() {parameterName=Refine1, defaultValue=<rgba 1 1 1 0.5>}
  n17 = TextureCoordinate()
  n18 = VectorParameter() {parameterName=Fire2_Distort_UVControl, channelNames {a=VPan, b=UPan, g=VScale, r=UScale}, defaultValue=<rgba 1 1 1 1>}
  n19 = AppendVector(A=@n18.UScale, B=@n18.UScale)
  n20 = TextureSampleParameter2D(UVs=Panner(Coordinate=Multiply(A=@n17, B=@n19), Speed=AppendVector(A=@n18.UPan, B=@n18.UPan))) {parameterName=Fire2_Distort_Texture, texture=&T_SoftNoise_BC7}
  n21 = AppendVector(A=@n20.R, B=@n20.R)
  n22 = ScalarParameter() {defaultValue=0.05, parameterName=Distort2_Mult}
  n23 = VectorParameter() {parameterName=Fire2_Texture_UVControl, channelNames {a=VPan, b=UPan, g=VScale, r=UScale}, defaultValue=<rgba 1 1 0.1 1>}
  n24 = Multiply(A=TextureCoordinate(), B=AppendVector(A=@n23.UScale, B=@n23.UScale))
  n25 = AppendVector(A=@n23.UPan, B=@n23.UPan)
  n26 = TextureSampleParameter2D(UVs=Add(A=Multiply(A=@n21, B=@n22), B=Panner(Coordinate=@n24, Speed=@n25))) {
   parameterName=Fire2_Texture
   samplerType=SAMPLERTYPE_LinearGrayscale
   texture=&T_FluidFlameTile_Greyscale
  }
  n27 = VectorParameter() {parameterName=Refine2, defaultValue=<rgba 1 1 1 0.5>}
  n28 = Saturate(None=Add(
    A=MaterialFunctionCall(Texture=@n15.RGB, RGB=@n16.RGB, A=@n16.RGB) {materialFunction=&MF_FX_Refine}.Result
    B=MaterialFunctionCall(Texture=@n26.RGB, RGB=@n27.RGB, A=@n27.RGB) {materialFunction=&MF_FX_Refine}.Result
   ))
  n29 = ScalarParameter() {group=WhiteWater2, parameterName=FireColor_Power}
  n30 = Multiply(
   A=LinearInterpolate(
    A=VectorParameter() {parameterName=BaseColor, defaultValue=<rgba 0 0 0 1>}.RGB
    B=VectorParameter() {parameterName=BaseColor2, defaultValue=<rgba 0 0 0 1>}.RGB
    Alpha=Power(Base=@n28, Exp=@n29)
   )
   B=LinearInterpolate(
    A=VectorParameter() {parameterName=EmmisiveMiddle, defaultValue=<rgba 0 15 0 1>}.RGB
    B=VectorParameter() {parameterName=Emmisive, defaultValue=<rgba 0 15 0 1>}.RGB
    Alpha=Fresnel() {exponent=2}
   )
  )
  n31 = OneMinus(None=Fresnel(ExponentIn=ScalarParameter() {defaultValue=0.1, parameterName="Exponent (S)"}))
  n32 = Multiply(
   A=@n30
   B=MaterialFunctionCall(Hue Shift Percentage=LinearInterpolate(Alpha=@n31) {constB=0.2}, Texture=@n30) {materialFunction=&HueShift}.Result
  )
  Desaturation(
   None=MaterialFunctionCall(Emissive=MaterialFunctionCall(Hue Shift Percentage=@n5, Texture=@n32) {materialFunction=&HueShift}.Result) {materialFunction=&MF_FX_EmissiveAdjustments}.Result
   Fraction=ScalarParameter() {parameterName=Desaturation}
  )
  MaterialFunctionCall!(Sine Phase=Time(), Value 1, Value 2) {materialFunction=/Engine/Functions/Engine_MaterialFunctions02/Utility/Sine_Remapped}
  Multiply()
  n33 = Multiply(
   A=Multiply(
    A=MaterialFunctionCall(Value=@n4.RGB) {materialFunction=&LinearSine}."Rounded Linear Sine "
    B=VertexNormalWS()
   )
   B=@n4.RGB
  )
  n34 = VectorParameter() {group=WPO, parameterName=WPO_GrayAdjust_Mul, defaultValue=<rgba 1 1 1 1>}
  n35 = TextureCoordinate()
  n36 = ComponentMask(None=@n35) {r=true}
  n37 = Multiply(A=ComponentMask(None=@n35) {g=true}, B=ScalarParameter() {defaultValue=1, parameterName=WPOGradient})
  n38 = TextureSample(UVs=Panner(Coordinate=AppendVector(A=@n36, B=@n37), Time=Time()) {speedY=1}) {texture=/Game/Effects/Fort_Effects/Textures/Ramps/T_Power_Source_Sphere_Ramp}
  n39 = VectorParameter() {parameterName=WPOGradient, defaultValue=<rgba 1 1 1 1>}
  Multiply(A=Multiply(A=Multiply(A=@n33, B=@n34.RGB), B=@n4.RGB), B=Multiply(A=VertexNormalWS(), B=Multiply(A=@n38.G, B=@n39.RGB)))
  n40 = ObjectPositionWS()
  n41 = Subtract(A=ComponentMask(None=@n40) {r=true}, B=ComponentMask(None=@n40) {g=true})
  n42 = ComponentMask(None=@n40) {b=true}
  NamedRerouteDeclaration(None=Frac(None=Subtract(A=@n41, B=@n42))) {name=ObjPosRandom, nodeColor=<rgba 0.5938 1 0 1>}
  n43 = LinearInterpolate(
   A=ScalarParameter() {defaultValue=0.25, group=WhiteWater2, parameterName=OpacityTexture_Low}
   B=ScalarParameter() {defaultValue=1, group=WhiteWater2, parameterName=OpacityTexture_High}
   Alpha=@n28
  )
  n44 = Multiply(
   A=ScalarParameter() {defaultValue=1, parameterName=FinalOpactiy}
   B=Power(Base=Multiply(A=@n43, B=@n31), Exp=ScalarParameter() {defaultValue=1, parameterName=OpacityPower})
  )
  n45 = DepthFade(FadeDistance=ScalarParameter() {defaultValue=100, parameterName=DepthFadeDistance})
  n46 = Constant() {r=1}
  n47 = StaticSwitchParameter(
   True=Multiply(
    A=@n44
    B=MaterialFunctionCall!("Xb1,PS4"=@n45, "PC-High&&Epic,XSX,PS5"=@n45, "PC-Perf,Android"=@n46, Switch=@n46, "PC-Low(NotPerf)&&Medium"=@n46) {materialFunction=/Game/Effects/Niagara/MasterMaterials/Functions/Common/PlatformSwitch_4D}.Result
   )
   False=@n44
  ) {parameterName=UseDepthFade?}
  n48 = MaterialFunctionCall() {materialFunction=&LinearGradient}
  n49 = ScalarParameter() {defaultValue=2, parameterName=LinGradientPower}
  StaticSwitchParameter(True=Multiply(A=@n47, B=Power(Base=@n48.VGradient, Exp=@n49)), False=@n47) {parameterName=UseTopGradient?}
 ]
}
