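# Unlit translucent material for drifting, glowing Hades skull sprites on instanced meshes, apparently driven by a grass actor.
# Motion math combines looping directional travel, sine drift, axis rotation, optional curl noise and normal-based scaling, varied by UV1 and instance.
# Color and opacity chains use hue-shifted tint, noise-distorted radial textures, pulsing, fake sun shading, and lifecycle, near-camera and depth fades.
# Some branches, such as time-of-day color and the height-driven curve offset, feed no other node.
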Material /Game/Environments/Helios/FX/Hades/Materials/M_Hades_Grassactor_Skulls1
graph {
 properties {
  bUsedWithInstancedStaticMeshes=true
  bUsedWithNanite=true
  blendMode=BLEND_Translucent
  opacityMaskClipValue=0.1
  pixelDepthOffsetMode=PDOM_Legacy
  previewMesh=/Game/Environments/Helios/FX/Hades/Meshes/S_FX_Hades_Skull_Niagara_01
  shadingModel=MSM_Unlit
  usedShadingModels=Unlit
 }
 nodes [
  ActorPositionWS()
  Add(
   A=NamedRerouteUsage()
   B=SphereMask(
    A=ObjectPositionWS()
    B=Constant3Vector()
    Radius=ScalarParameter() {defaultValue=100, parameterName=Sphere_Radius}
    Hardness=ScalarParameter() {parameterName=Sphere_Hardness}
   )
  )
  n1 = Multiply(A=Time(), B=ScalarParameter() {defaultValue=0.001, parameterName=VarianceSpeed}) {constB=0.05}
  n2 = ComponentMask(None=TextureCoordinate() {coordinateIndex=1}) {r=true}
  n3 = Add(A=Multiply(A=Time(), B=ScalarParameter() {defaultValue=0.4, parameterName=DriftSpeed}), B=Sine(None=Add(A=@n1, B=@n2)) {period=0.17})
  n4 = Sine(None=@n3) {period=0.56}
  n5 = Sine(None=@n3) {period=0.7}
  n6 = Add(A=@n1, B=@n2)
  n7 = Multiply(
   A=AppendVector(A=AppendVector(A=@n4, B=@n5), B=Sine(None=@n3) {period=0.82})
   B=LinearInterpolate(
    A=ScalarParameter() {defaultValue=1, parameterName=MinHorizontalDrift}
    B=ScalarParameter() {defaultValue=1, parameterName=MaxHorizontalDrift}
    Alpha=Sine(None=@n6) {period=0.14}
   )
  )
  n8 = Multiply(A=ObjectRadius(), B=Constant() {r=0.5}) {constB=0.5}
  n9 = Multiply(A=@n8) {constB=-1}
  n10 = Time() {bOverride_Period=true, period=3}
  n11 = Time() {bOverride_Period=true, period=8}
  n12 = ComponentMask(None=TextureCoordinate() {coordinateIndex=1}) {g=true}
  n13 = Add(
   A=StaticSwitchParameter(True=Divide(A=@n10) {constB=3}, False=Divide(A=@n11) {constB=8}) {defaultValue=true, parameterName=UseFastSprites}
   B=Multiply(A=Subtract(A=@n12) {constB=0.5}) {constB=-2}
  )
  n14 = Frac(None=@n13)
  n15 = ScalarParameter() {defaultValue=1, parameterName=OverallSpeed}
  n16 = ScalarParameter() {defaultValue=1, parameterName=OverallSpeed_Low}
  n17 = ComponentMask(None=TextureCoordinate() {coordinateIndex=1}) {g=true}
  n18 = Multiply(
   A=Multiply(A=LinearInterpolate(A=@n9, B=@n8, Alpha=@n14), B=LinearInterpolate(A=@n15, B=@n16, Alpha=@n17))
   B=LinearInterpolate(
    A=VectorParameter() {parameterName=OverallDirection, defaultValue=<rgba 1 0 0 1>}.RGB
    B=VectorParameter() {parameterName=OverallDirection_Low, defaultValue=<rgba 1 0 0 1>}.RGB
    Alpha=@n17
   )
  )
  n19 = VectorParameter() {parameterName=RotationAxis}
  n20 = ObjectPositionWS()
  n21 = Add(A=Time(), B=ComponentMask(None=@n20) {b=true})
  n22 = ObjectPositionWS()
  n23 = LinearInterpolate(
   A=ScalarParameter() {defaultValue=1, parameterName=RotationSpeed}
   B=ScalarParameter() {defaultValue=1, parameterName=RotationSpeed_Low}
   Alpha=Frac(None=ComponentMask(None=@n22) {b=true})
  )
  n24 = Add(
   A=Add(A=Constant(), B=Add(A=@n7, B=@n18))
   B=RotateAboutAxis(
    NormalizedRotationAxis=Transform(None=@n19.RGB) {transformSourceType=TRANSFORMSOURCE_Local}
    RotationAngle=Multiply(A=@n21, B=@n23)
    PivotPoint=@n20
    Position=WorldPosition().XYZ
   )
  )
  n25 = Add(A=@n24, B=NamedRerouteUsage())
  n26 = VectorParameter() {parameterName=SkullOffsetAt0, defaultValue=<rgba 0 0 150 1>}
  n27 = ScalarParameter() {parameterName=Scale_Low}
  n28 = ScalarParameter() {parameterName=Scale_High}
  n29 = PerInstanceRandom()
  n30 = TextureCoordinate() {coordinateIndex=1}
  n31 = Clamp(None=Add(A=LinearInterpolate(Alpha=@n29) {constB=0.5}, B=ComponentMask(None=@n30) {g=true}))
  Add(
   A=Add(A=StaticSwitchParameter(True=@n25, False=@n24) {parameterName=UseCurl?}, B=@n26.RGB)
   B=Multiply(A=LinearInterpolate(A=@n27, B=@n28, Alpha=@n31), B=VertexNormalWS())
  )
  n32 = ObjectPositionWS()
  n33 = Divide(A=ComponentMask(None=Add(A=@n32) {constB=0}) {b=true}, B=ScalarParameter() {defaultValue=2680, parameterName=Param_1})
  Clamp(None=@n33)
  Constant() {r=1}
  Constant()
  Constant() {r=1}
  Constant() {r=2}
  Constant() {r=1}
  Constant() {r=1}
  Constant() {r=1}
  LinearInterpolate()
  n34 = CurveAtlasRowParameter(CurveTime=@n33) {
   atlas=/Game/Environments/Helios/FX/Hades/Textures/Hades_Atlas
   curve=/Game/Environments/Helios/FX/Hades/Textures/Hades_Skulls_Curve
   parameterName=CurveAtlas
  }
  LinearInterpolate(A=@n26.RGB, B=VectorParameter() {parameterName=SkullOffsetAtTop, defaultValue=<rgba 0 0 25 1>}.RGB, Alpha=@n34.G)
  MaterialFunctionCall() {materialFunction=&LocalPosition}
  n35 = ObjectPositionWS()
  Multiply(A=PerInstanceRandom(), B=Frac(None=ComponentMask(None=@n35) {b=true}))
  n36 = VertexColor()
  n37 = Multiply(A=Constant() {r=1}, B=NamedRerouteUsage())
  n38 = ScalarParameter() {defaultValue=5, parameterName=Emmisive}
  Multiply(A=Multiply(A=Multiply(A=@n36.G), B=Multiply(A=@n37, B=@n38)), B=NamedRerouteUsage())
  n39 = Multiply(
   A=Constant() {r=1}
   B=MaterialFunctionCall(Distance=ScalarParameter() {defaultValue=500, parameterName=NearCamShrinkDist}) {materialFunction=&Near_Camera_Fade}.Result
  )
  n40 = Multiply(A=@n39)
  n41 = Multiply(A=@n39)
  n42 = ScalarParameter() {parameterName=UseExtraFade?}
  n43 = Add(A=@n13) {constB=-0.25}
  n44 = LinearInterpolate(Alpha=ConstantBiasScale(None=Sine(None=@n43)))
  n45 = Multiply(
   A=Multiply(A=LinearInterpolate(A=@n40, B=@n41, Alpha=@n42), B=PerInstanceFadeAmount())
   B=Multiply(A=Power(Base=@n44) {constExponent=0.6}, B=NamedRerouteUsage())
  )
  n46 = NamedRerouteUsage()
  Multiply(A=Multiply(A=ScalarParameter() {defaultValue=5, parameterName=Opacity}, B=Multiply(A=@n45, B=@n46)), B=DepthFade())
  Multiply()
  n47 = TextureSample(UVs=Add(A=TextureCoordinate() {coordinateIndex=2, uTiling=0.65, vTiling=1.1}, B=Constant2Vector() {r=0.15})) {texture=/Game/Environments/Helios/FX/Hades/Textures/T_Skull_White_D_01}
  Multiply(A=@n47.RGB, B=@n47.RGB)
  n48 = CollectionParameter() {
   collection=&FortniteMaterialParameters
   parameterName=SunAndMoonModelDirectionalVector
  }
  n49 = DotProduct(A=ComponentMask(None=@n48) {b=true, g=true, r=true}, B=ReflectionVectorWS())
  NamedRerouteDeclaration(None=LinearInterpolate(
    A=ScalarParameter() {defaultValue=0.5, group=Texture, parameterName=SunSide_Mult}
    B=ScalarParameter() {defaultValue=0.35, group=Texture, parameterName=ShadowSide_Mult}
    Alpha=Clamp(None=@n49)
   )) {name=FakeSunShadowing, nodeColor=<rgba 1 0.3866 0 1>}
  n50 = Time()
  n51 = TextureCoordinate() {coordinateIndex=1}
  n52 = Multiply(
   A=LinearInterpolate(Alpha=PerInstanceRandom()) {constA=0.5}
   B=LinearInterpolate(
    A=ScalarParameter() {defaultValue=1, parameterName=PulseSpeed}
    B=ScalarParameter() {defaultValue=0.5, parameterName=PulseSpeedSlow}
    Alpha=ComponentMask(None=@n51) {r=true}
   ) {constA=0.5}
  )
  n53 = Power(Base=MaterialFunctionCall!(Sine Phase=Multiply(A=@n50, B=@n52), Value 1=Constant(), Value 2=Constant() {r=1}) {materialFunction=/Engine/Functions/Engine_MaterialFunctions02/Utility/Sine_Remapped}.Remapped Sine) {constExponent=10}
  NamedRerouteDeclaration(None=LinearInterpolate(Alpha=Saturate(None=@n53)) {constA=0.25, constB=2}) {name=Pulse, nodeColor=<rgba 1 0.4453 0 1>}
  n54 = TextureCoordinate()
  n55 = Constant2Vector() {g=0.5, r=0.5}
  n56 = ScalarParameter() {defaultValue=0.5, parameterName=CenterDot_Radius}
  n57 = ScalarParameter() {parameterName=CenterDot_Hardness}
  NamedRerouteDeclaration(None=Multiply(A=SphereMask(A=@n54, B=@n55, Radius=@n56, Hardness=@n57), B=ScalarParameter() {defaultValue=1, parameterName=CenterDot_OpacitytMult})) {name=Centerdot, nodeColor=<rgba 0 0.2734 1 1>}
  n58 = VertexColor()
  n59 = VertexColor()
  n60 = LinearInterpolate(
   A=ScalarParameter() {defaultValue=50, parameterName=CurlIntensity}
   B=ScalarParameter() {defaultValue=100, parameterName=CurlIntensityHigh}
   Alpha=@n59.G
  )
  n61 = Subtract(A=WorldPosition().XYZ, B=ObjectPositionWS())
  n62 = Multiply(A=Time(), B=ScalarParameter() {defaultValue=1, parameterName=CurlSpeed})
  n63 = ScalarParameter() {defaultValue=1000, parameterName=CurlScale}
  n64 = ScalarParameter() {defaultValue=500, parameterName=CurlScale_Low}
  n65 = VectorNoise(World Position=Divide(A=Subtract(A=@n61, B=@n62), B=LinearInterpolate(A=@n63, B=@n64, Alpha=@n59.G))) {noiseFunction=VNF_VectorALU}
  NamedRerouteDeclaration(None=Multiply(A=OneMinus(None=@n58.A), B=Multiply(A=@n60, B=@n65))) {name=Curl, nodeColor=<rgba 0.5234 0 1 1>}
  n66 = Fresnel(ExponentIn=ScalarParameter() {defaultValue=1, parameterName=ExponentIn}) {baseReflectFraction=0.0001, exponent=0.1}
  NamedRerouteDeclaration(None=Power(Base=OneMinus(None=@n66))) {name=Fresnel, nodeColor=<rgba 0.0781 0 1 1>}
  n67 = MaterialFunctionCall() {materialFunction=&VectorToRadialValue}
  n68 = MaterialFunctionCall(Float2=@n67.Radial Coordinates) {materialFunction=&BreakOutFloat2Components}
  n69 = ScalarParameter() {defaultValue=0.1, parameterName=MaskOffsetLow}
  n70 = ScalarParameter() {defaultValue=0.1, parameterName=MaskOffset}
  n71 = PerInstanceRandom()
  n72 = Multiply(
   A=Subtract(A=Multiply(A=@n68.G), B=LinearInterpolate(A=@n69, B=@n70, Alpha=@n71))
   B=ScalarParameter() {defaultValue=0.2, parameterName=MaskScale}
  )
  n73 = PerInstanceRandom()
  n74 = AppendVector(A=@n73, B=@n73)
  n75 = VectorParameter() {parameterName=Speed, defaultValue=<rgba -0.1 -0.2 2 2>}
  n76 = MaterialFunctionCall(UVs=@n67.Radial Coordinates, Texture Scale=AppendVector(A=@n75.B, B=@n75.B)) {materialFunction=&ScaleUVsByCenter}
  n77 = Add(A=@n75.R, B=LinearInterpolate(Alpha=PerInstanceRandom()) {constB=0.1})
  n78 = Add(A=@n75.G, B=LinearInterpolate(Alpha=PerInstanceRandom()) {constB=0.1})
  n79 = TextureSample(UVs=Panner(Coordinate=Add(A=@n74, B=@n76.UVs), Speed=AppendVector(A=@n77, B=@n78))) {samplerType=SAMPLERTYPE_Normal, texture=/Game/VisualThreatManager/StormVisuals/LargeClouds/T_StormSwirlNormal}
  n80 = AppendVector(A=@n79.R, B=@n79.R)
  n81 = ScalarParameter() {defaultValue=1, parameterName=Noise2}
  n82 = Add(A=TextureCoordinate(), B=Multiply(A=Saturate(None=@n72), B=Multiply(A=@n80, B=@n81)))
  n83 = TextureSample(UVs=@n82) {texture=/Game/Global/Textures/Default/Masks_and_Gradients/T_New_Soft_Flair}
  n84 = Add(A=@n82, B=LinearInterpolate(Alpha=PerInstanceRandom()) {constB=0.2})
  n85 = Multiply(
   A=Add(A=@n83.RGB, B=TextureSample(UVs=@n84) {texture=/Game/Effects/Fort_Effects/Textures/Tile/T_ShieldCaustics_Tile_Chromatic}.RGB)
   B=SphereMask(A=TextureCoordinate(), B=Constant2Vector() {g=0.5, r=0.5}, Radius=ScalarParameter() {defaultValue=0.4, parameterName=Rad}) {hardnessPercent=0}
  )
  NamedRerouteDeclaration(None=Multiply(A=@n85, B=ScalarParameter() {defaultValue=1, parameterName=Emissive})) {name=AnimatedTexture, nodeColor=<rgba 0.8516 1 0 1>}
  NamedRerouteDeclaration(None=@n85) {name=OpacityAnimatedTexture, nodeColor=<rgba 0.6641 1 0 1>}
  n86 = LinearInterpolate(A=ScalarParameter() {parameterName=HueShiftAmount}, B=ScalarParameter() {parameterName=HueShiftAmountMax}, Alpha=PerInstanceRandom())
  n87 = Multiply(A=ScalarParameter() {parameterName=HueShiftAmount}, B=Multiply(A=VertexColor().R, B=@n86)) {constB=-0.05}
  NamedRerouteDeclaration(None=MaterialFunctionCall(
    Hue Shift Percentage=StaticSwitchParameter(True=@n87, False=@n86) {defaultValue=true, parameterName=UseVertexForHueShift?}
    Texture=VectorParameter() {parameterName="Color_Day (V3)", defaultValue=<rgba 0 1 0.1012 1>}.RGB
   ) {materialFunction=&HueShift}.Result) {name=BaseColorTOD, nodeColor=<rgba 0.7578 1 0 1>}
  ObjectPositionWS()
  PerInstanceRandom()
  Power(Base=@n83.RGB)
  Saturate(None=ScalarParameter() {defaultValue=1, parameterName=SkullAnimation})
  ScalarParameter()
  StaticSwitchParameter() {parameterName=IsGrassMotes?}
  StaticSwitchParameter(
   True=MaterialFunctionCall() {materialFunction=/Game/Effects/Fort_Effects/Materials/_Functions/CurveColor_Time_of_Day}.Result
   False=MaterialFunctionCall(
    Color_Morning=VectorParameter() {parameterName="Color_Morning (V3)", defaultValue=<rgba 0 1 0.1012 1>}.RGB
    Color_Day=VectorParameter() {parameterName="Color_Day (V3)", defaultValue=<rgba 0 1 0.1012 1>}.RGB
    Color_Evening=VectorParameter() {parameterName="Color_Evening (V3)", defaultValue=<rgba 0 1 0.1012 1>}.RGB
    Color_Night=VectorParameter() {parameterName="Color_Night (V3)", defaultValue=<rgba 0 1 0.1012 1>}.RGB
   ) {materialFunction=&Color_Time_of_Day}.Result
  ) {parameterName=UseCurveforTODinsteadofScalar?}
  VertexColor()
  VertexInterpolator(VS=@n17)
 ]
}
