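# Masked, two-sided subsurface material for Niagara mesh-particle ficus leaves in the Hades effects, named for leaves on the ground.
# Particle color is scaled and hue-shifted through dynamic parameters, while particle alpha multiplies a half-tiled leaf texture's alpha.
# The graph also holds a particle-position synced distance fade, normal and packed textures, and roughness, opacity and subsurface-color parameters.
# A named SunDirectionGradient squares the saturated dot of camera vector and negated sunlight direction.
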Material /Game/Environments/Helios/FX/Hades/Materials/M_Hades_Leaves_OnGround_MeshParticle
graph {
 properties {
  bUsedWithNiagaraMeshParticles=true
  blendMode=BLEND_Masked
  pixelDepthOffsetMode=PDOM_Legacy
  shadingModel=MSM_Subsurface
  twoSided=true
  usedShadingModels=Subsurface
 }
 nodes [
  n1 = ParticleColor()
  MaterialFunctionCall(
   Hue Shift Percentage=DynamicParameter() {paramNames [ColorMult,HueShiftMult,Param3,Param4]}.HueShiftMult
   Texture=Multiply(A=DynamicParameter() {paramNames [ColorMult,HueShiftMult,Param3,Param4]}.ColorMult, B=@n1.RGB)
  ) {materialFunction=&HueShift}
  MaterialFunctionCall(Position=ParticlePositionWS()) {materialFunction=&MF_Distance_SyncedFade_FX}
  Multiply(
   A=TextureSample(UVs=TextureCoordinate() {uTiling=0.5, vTiling=0.5}) {texture=/Game/Environments/Helios/FX/Hades/Textures/T_FicusTree_01_Leaves_MeshParticle_SSSR}.A
   B=@n1.A
  )
  Multiply()
  n2 = CollectionParameter() {
   collection=&FortniteMaterialParameters
   parameterName=SunlightDirectionalVector
  }
  n3 = ComponentMask(None=@n2) {b=true, g=true, r=true}
  n4 = Saturate(None=DotProduct(A=CameraVectorWS(), B=Multiply(A=@n3) {constB=-1}))
  NamedRerouteDeclaration(None=Multiply(A=@n4, B=@n4)) {name=SunDirectionGradient, nodeColor=<rgba 0.2422 0 1 1>}
  ScalarParameter() {defaultValue=0.5, parameterName=roughness}
  ScalarParameter() {defaultValue=1, parameterName=Opacity}
  TextureSample(UVs=TextureCoordinate() {uTiling=0.5, vTiling=0.5}) {samplerType=SAMPLERTYPE_Normal, texture=/Game/Environments/Helios/FX/Hades/Textures/T_FicusTree_01_Leaves_MeshParticle_N}
  TextureSample() {texture=/Game/Environments/Helios/FX/Hades/Textures/T_FicusTree_01_Leaves_MeshParticle_OSSR}
  TextureSample() {texture=&T_FicusTree_01_Leaves_MeshParticle_SSSR}
  TextureSample() {texture=&T_FicusTree_01_Leaves_MeshParticle_OSSR}
  VectorParameter() {parameterName=subsurface color, defaultValue=<rgba 0.0217 0.0164 0.0292 1>}
 ]
}
