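# Translucent, two-sided material for drifting smoke plumes on instanced meshes in the Helios Hades environment, tinted by time-of-day emissive color.
# Opacity combines a rotated, panning smoke texture with storm-cloud masks, offset by position and instance randomness, then shaped by vertical, near-camera, depth, distance and instance fades.
# SplineThicken sets plume width, shrinking near the camera, and panning distortion textures add world-position waves.
# Quality and shading-path switches drop depth fade, detail blending or waves on cheaper paths.
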Material /Game/Environments/Helios/FX/Hades/Materials/M_Hades_Smoke_Plumes_GrassActor
graph {
 properties {
  bUsedWithInstancedStaticMeshes=true
  blendMode=BLEND_Translucent
  pixelDepthOffsetMode=PDOM_Legacy
  twoSided=true
  usedShadingModels=Default Lit
 }
 nodes [
  n1 = ScalarParameter() {defaultValue=25, group=Texture, parameterName=OpacityFinal}
  n2 = Multiply(
   A=ComponentMask(None=TextureCoordinate()) {g=true}
   B=AppendVector(A=Constant(), B=ScalarParameter() {defaultValue=0.5, parameterName=HeightUVScaleAdd})
  )
  n3 = VectorParameter() {parameterName=TileDetail, defaultValue=<rgba 4 2.5 0 0>}
  n4 = Multiply(A=Add(A=NamedRerouteUsage(), B=TextureCoordinate()), B=Add(A=@n2, B=AppendVector(A=@n3.R, B=@n3.R)))
  n5 = ComponentMask(None=VectorParameter() {parameterName=PanSpeedDetail, defaultValue=<rgba -0.015 0.025 0 1>}.RGB) {g=true, r=true}
  n6 = TextureSampleParameter2D(UVs=VertexInterpolator(VS=Panner(Coordinate=@n4, Speed=@n5) {bFractionalPart=true}).PS) {
   parameterName=MaskDetail
   samplerType=SAMPLERTYPE_LinearGrayscale
   texture=/Game/Characters/Player/Male/Medium/Bodies/M_MED_Hurricane/Textures/T_MED_Hurricane_StormCloud
  }
  n7 = VectorParameter() {parameterName=TileMain, defaultValue=<rgba 2 1 0 0>}
  n8 = Multiply(A=Add(A=NamedRerouteUsage(), B=TextureCoordinate()), B=Add(A=AppendVector(A=@n7.R, B=@n7.R), B=@n2))
  n9 = ComponentMask(None=VectorParameter() {parameterName=PanSpeedMain, defaultValue=<rgba -0.01 0.005 0 1>}.RGB) {g=true, r=true}
  n10 = TextureSampleParameter2D(UVs=VertexInterpolator(VS=Panner(Coordinate=@n8, Speed=@n9) {bFractionalPart=true}).PS) {
   parameterName=MaskMain
   samplerType=SAMPLERTYPE_LinearGrayscale
   texture=/Game/Global/Textures/Default/Tileable_Patterns/T_StormCloudDetail01_M
  }
  n11 = Power(Base=@n10.G, Exp=ScalarParameter() {defaultValue=0.6, parameterName=DetailLerpPower})
  n12 = ScalarParameter() {defaultValue=5000, parameterName=DepthFadeDistance}
  n13 = Constant() {r=1}
  n14 = ScalarParameter() {defaultValue=0.5, group=Texture, parameterName=TilingSmoke_V_Mult}
  n15 = TextureCoordinate() {vTiling=0.35}
  n16 = ComponentMask(None=@n15) {g=true}
  n17 = LinearInterpolate(Alpha=VertexColor().R) {constA=1.25, constB=2}
  n18 = NamedRerouteUsage()
  n19 = AppendVector(A=ComponentMask(None=@n15) {r=true}, B=Add(A=Multiply(A=@n16, B=@n17), B=ComponentMask(None=@n18) {r=true}))
  n20 = AppendVector(A=Constant(), B=Frac(None=PerInstanceRandom()))
  n21 = Add(A=NamedRerouteUsage(), B=Multiply(A=AppendVector(A=@n13, B=@n14), B=Add(A=@n19, B=@n20)))
  n22 = ComponentMask(None=VectorParameter() {group=Texture, parameterName=PannerSpeed, defaultValue=<rgba 0 0.035 0 1>}.RGB) {g=true, r=true}
  n23 = VertexInterpolator(VS=MaterialFunctionCall(
    UVs=Panner(Coordinate=@n21, Speed=@n22)
    "Rotation Angle (0-1)"=ScalarParameter() {defaultValue=0.25, group=Texture, parameterName="Rotation Angle (0-1) (S)"}
   ) {materialFunction=&CustomRotator}.Rotated Values)
  n24 = ScalarParameter() {defaultValue=500, parameterName=NearFade}
  n25 = TextureCoordinate()
  n26 = ComponentMask(None=@n25) {g=true}
  n27 = ComponentMask(None=@n25) {g=true}
  n28 = Multiply(
   A=Multiply(
    A=TextureSampleParameter2D(UVs=@n23.PS) {parameterName=SmokeTexture_01, texture=/Game/Effects/Fort_Effects/Textures/Smoke/T_Puffy_Smoke_Trail_01}.G
    B=MaterialFunctionCall(Distance=@n24) {materialFunction=&Near_Camera_Fade}.Result
   )
   B=Multiply(A=Multiply(A=@n26, B=@n26), B=OneMinus(None=@n27))
  )
  n29 = QualitySwitch(Default=Multiply(A=DepthFade(FadeDistance=@n12), B=@n28), Low=@n28)
  n30 = Multiply(
   A=ShadingPathSwitch(Default=LinearInterpolate(A=@n6.G, B=@n10.G, Alpha=@n11), Forward=@n10.G, Mobile=@n10.G)
   B=ShadingPathSwitch(Default=@n29, Deferred=@n29, Forward=@n28)
  )
  Multiply(
   A=Multiply(A=PerInstanceFadeAmount(), B=Multiply(A=@n1, B=@n30))
   B=MaterialFunctionCall(
    Fade_Start_Distance=ScalarParameter() {defaultValue=5000, parameterName="Fade_Start_Distance (S)"}
    Fade_End_Distance=ScalarParameter() {defaultValue=1500, parameterName="Fade_End_Distance (S)"}
   ) {materialFunction=&MF_Distance_SyncedFade_FX}.Result
  )
  Multiply(
   A=ScalarParameter() {defaultValue=1, parameterName=EmissiveAmount}
   B=MaterialFunctionCall(
    Color_Morning=VectorParameter() {parameterName="Color_Morning (V3)", defaultValue=<rgba 1 1 1 0>}.RGB
    Color_Day=VectorParameter() {parameterName="Color_Day (V3)", defaultValue=<rgba 1 1 1 0>}.RGB
    Color_Evening=VectorParameter() {parameterName="Color_Evening (V3)", defaultValue=<rgba 1 1 1 0>}.RGB
    Color_Night=VectorParameter() {parameterName="Color_Night (V3)", defaultValue=<rgba 1 1 1 0>}.RGB
   ) {materialFunction=&Color_Time_of_Day}.Result
  )
  n31 = Constant()
  n32 = Transform(None=ObjectPositionWS()) {transformSourceType=TRANSFORMSOURCE_World, transformType=TRANSFORM_Local}
  n33 = Multiply(A=ComponentMask(None=@n32) {r=true}, B=ComponentMask(None=@n32) {g=true})
  NamedRerouteDeclaration(None=Frac(None=AppendVector(A=@n31, B=@n33))) {name=OffsetUvBasedonPos, nodeColor=<rgba 0 1 0.8359 1>}
  n34 = MaterialFunctionCall(Distance=ScalarParameter() {defaultValue=500, parameterName=NearFade}) {materialFunction=&Near_Camera_Fade}
  n35 = MaterialFunctionCall(
   WidthBase=Multiply(A=ScalarParameter() {defaultValue=15, group=WPO, parameterName=WidthBase_Max}, B=@n34.Result)
   WidthTip=Multiply(A=@n34.Result, B=ScalarParameter() {defaultValue=25, group=WPO, parameterName=WidthTip_Max})
  ) {materialFunction=&SplineThicken}
  n36 = Constant3Vector() {desc=Disable WPO anim in Low, constant=<rgba 0 0 20 0>}
  n37 = Constant3Vector()
  n38 = VectorParameter() {group=WPO, parameterName=TextureWPO_UVtiling, defaultValue=<rgba 1 2 0 0>}
  n39 = VectorParameter() {group=WPO, parameterName=TextureWPO_UVSpeed, defaultValue=<rgba 0 0.15 0 0>}
  n40 = ObjectPositionWS()
  n41 = ComponentMask(None=@n40) {r=true}
  n42 = ComponentMask(None=@n40) {g=true}
  n43 = Multiply(B=Frac(None=Subtract(A=@n41, B=@n42))) {constA=0.1}
  n44 = Panner(
   Coordinate=Multiply(A=TextureCoordinate(), B=ComponentMask(None=@n38.RGB) {g=true, r=true})
   Speed=Add(A=ComponentMask(None=@n39.RGB) {g=true, r=true}, B=Abs(None=@n43))
  ) {speedY=0.15}
  n45 = TextureSample(UVs=@n44) {samplerType=SAMPLERTYPE_Normal, texture=&T_Weap_Distort_N_02}
  n46 = ComponentMask(None=@n45.RGB) {r=true}
  n47 = ScalarParameter() {defaultValue=25, group=WPO, parameterName=WPO_HeightX}
  n48 = ComponentMask(None=@n45.RGB) {g=true}
  n49 = ScalarParameter() {group=WPO, parameterName=WPO_HeightY}
  n50 = AppendVector(
   A=AppendVector(A=Multiply(A=@n46, B=@n47) {constB=0.5}, B=Multiply(A=@n48, B=@n49) {constB=0.5})
   B=Multiply(
    A=TextureSample(UVs=@n44) {texture=&T_Smooth_tile}.G
    B=ScalarParameter() {group=WPO, parameterName=WPO_HeightZ}
   ) {constB=0.6}
  )
  n51 = Constant() {r=0.5}
  n52 = SphereMask(
   A=MaterialFunctionCall() {materialFunction=&LinearGradient}.VGradient
   B=@n51
   Radius=@n51
  ) {hardnessPercent=0}
  n53 = Multiply(
   A=Transform(None=LinearInterpolate(A=@n37, B=@n50, Alpha=@n52))
   B=ScalarParameter() {defaultValue=100, group=WPO, parameterName=WPO_Wave_Intenstiy}
  ) {constB=80}
  n54 = ShadingPathSwitch(Default=@n36, Deferred=QualitySwitch(Default=Abs(None=@n53), Low=@n36))
  ShadingPathSwitch(Default=@n35.WorldPositionOffset, Deferred=QualitySwitch(Default=Add(A=@n35.WorldPositionOffset, B=@n54), Low=@n35.WorldPositionOffset))
 ]
}
