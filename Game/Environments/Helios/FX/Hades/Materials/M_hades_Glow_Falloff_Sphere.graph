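# Additive unlit material for a Hades glow falloff sphere, with Fresnel-blended colors tinted by time of day.
# World position offset scales the sphere around its pivot by time of day, shrinking it as the camera nears.
# Opacity combines depth-faded Fresnel, a vertical UV mask and a distance fade that hides the glow up close.
# By default, inside the fade offset the mesh also collapses and drops far below its origin.
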Material /Game/Environments/Helios/FX/Hades/Materials/M_hades_Glow_Falloff_Sphere
graph {
 properties {blendMode=BLEND_Additive, pixelDepthOffsetMode=PDOM_Legacy, shadingModel=MSM_Unlit, usedShadingModels=Unlit}
 nodes [
  n1 = StaticSwitchParameter(
   True=ActorPositionWS()
   False=MaterialFunctionCall() {materialFunction=&ObjectPivotPoint}.Object Pivot Location
  ) {defaultValue=true, parameterName=UseActorPosOrUseObject?}
  n2 = Subtract(A=@n1, B=WorldPosition().XYZ)
  n3 = ScalarParameter() {defaultValue=0.84, group=Opacity, parameterName=Morning/EveningScale}
  n4 = ScalarParameter() {defaultValue=0.5, group=Opacity, parameterName=DayTimeScale}
  n5 = ScalarParameter() {defaultValue=0.5, group=Opacity, parameterName=NightScale}
  n6 = Distance(A=@n1, B=CameraPositionWS())
  n7 = ScalarParameter() {defaultValue=3000, group=Fade, parameterName=FadeDistanceOffset}
  n8 = Clamp(None=Divide(A=@n6, B=Multiply(A=@n7) {constB=4})) {minDefault=0.3}
  n9 = Clamp(None=Subtract(
    A=MaterialFunctionCall(Color_Morning=@n3, Color_Day=@n4, Color_Evening=@n3, Color_Night=@n5) {materialFunction=&Color_Time_of_Day}.Result
    B=LinearInterpolate(Alpha=@n8) {constA=0.7, constB=0}
   ))
  n10 = Subtract(
   A=Distance(A=ObjectPositionWS(), B=CameraPositionWS())
   B=ScalarParameter() {defaultValue=3000, group=Fade, parameterName=FadeDistanceOffset}
  )
  n11 = ScalarParameter() {defaultValue=12000, group=Fade, parameterName=FadeDistance}
  n12 = Clamp(None=Ceil(None=Divide(A=@n10, B=@n11)))
  n13 = OneMinus(None=@n12)
  n14 = Add(A=Subtract(A=ObjectPositionWS(), B=WorldPosition().XYZ), B=Constant3Vector() {constant=<rgba 0 0 -70000 0>})
  Add(
   A=StaticSwitchParameter(True=LinearInterpolate(A=@n2, Alpha=@n9) {constAlpha=0, constB=0}, False=Constant()) {defaultValue=true, parameterName=WPOScaling}
   B=StaticSwitchParameter(True=@n12, False=Multiply(A=@n13, B=@n14)) {parameterName=ShrinkWhenFadeOutForce?}
  )
  Constant() {r=1}
  n15 = VectorParameter() {group=Colors, parameterName=DayTint, defaultValue=<rgba 0.8 0.225 0.225 0>}
  n16 = MaterialFunctionCall(
   Color_Morning=@n15.RGB
   Color_Day=@n15.RGB
   Color_Evening=@n15.RGB
   Color_Night=VectorParameter() {group=Colors, parameterName=NightTint, defaultValue=<rgba 0.8 0.2985 0.0216 1>}.RGB
  ) {materialFunction=&Color_Time_of_Day}
  n17 = MaterialFunctionCall(
   Brightness=ScalarParameter() {defaultValue=1, parameterName=Fresnel Brightness}
   Exponent=ScalarParameter() {defaultValue=7, parameterName=Fresnel Exponent}
  ) {materialFunction=&MF_Fresnel_FX}
  n18 = LinearInterpolate(
   A=VectorParameter() {group=Colors, parameterName=GlowColorOuter, defaultValue=<rgba 0.7 0.1768 0.0435 0>}.RGB
   B=VectorParameter() {group=Colors, parameterName=GlowColor, defaultValue=<rgba 1 0.4882 0 0>}.RGB
   Alpha=@n17.Fresnel
  )
  MaterialFunctionCall(Emissive=MaterialFunctionCall!(StartingHue, Base Emissive=Multiply(A=@n16.Result, B=@n18), Offset) {materialFunction=/Game/Environments/Props/RiftBeacon/FX/RiftBubble_GreenGoop}.TacoTimeResults) {materialFunction=&MF_FX_EmissiveAdjustments}
  n19 = ScalarParameter() {defaultValue=0.2, group=Opacity, parameterName=OpacityMult}
  n20 = DepthFade(Opacity=@n17.Fresnel, FadeDistance=ScalarParameter() {defaultValue=2000, group=Fade, parameterName=DepthFade})
  n21 = ScalarParameter() {defaultValue=3000, group=Fade, parameterName=FadeDistanceOffset}
  n22 = Clamp(None=Divide(A=Subtract(A=@n6, B=@n21), B=ScalarParameter() {defaultValue=12000, group=Fade, parameterName=FadeDistance}))
  n23 = ComponentMask(None=OneMinus(None=TextureCoordinate())) {g=true}
  n24 = Clamp(None=Multiply(A=Power(Base=@n23) {constExponent=2.5}) {constB=35})
  Multiply(
   A=ScalarParameter() {defaultValue=1, group=Opacity, parameterName=FinalOpacity}
   B=Multiply(A=Multiply(A=@n19, B=@n20), B=Multiply(A=@n22, B=@n24))
  )
 ]
}
